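# Physics asset for the EclairAche medium male player body cosmetic, providing capsule bodies and constraints for a whistle bone chain and the upper spine.
# The dyn_whistle bones form a serial chain from a small master body through segments 01 to 07, with tighter swing and twist limits near the root.
# Larger capsules on spine_03 to 05 approximate the torso, with spine_04 and 05 both constrained to spine_03 with free angular motion.
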PhysicsAsset /BRCosmetics/Characters/Player/Male/Medium/Bodies/M_MED_EclairAche/Meshes/M_MED_EclairAche_Physics
bodies {
 dyn_whistle_01 [{
   shapeType=Capsule
   length=10.4147
   radius=3.4444
   shapeName=dyn_whistle_01_capsule
   center=<3.475 -0.0004 2.3695>
   rotation=<rot 0 179.9896 -89.9958>
  }]
 dyn_whistle_02 [{
   shapeType=Capsule
   length=11.2652
   radius=2.3585
   shapeName=dyn_whistle_02_capsule
   center=<1.8933 0.1324 2.0184>
   rotation=<rot 0.0002 0 -90>
  }]
 dyn_whistle_03 [{
   shapeType=Capsule
   length=8.0211
   radius=2
   shapeName=dyn_whistle_03_capsule
   center=<-0.0006 0.1132 1.002>
   rotation=<rot -0.1105 1.267 -89.9511>
  }]
 dyn_whistle_04 [{shapeType=Capsule, length=5.86, radius=2, shapeName=dyn_whistle_04_capsule, center=<0.1469 0.1146 0.9465>, rotation=<rot 0.0001 0 -91.37>}]
 dyn_whistle_05 [{shapeType=Capsule, length=0.5472, radius=2, shapeName=dyn_whistle_05_capsule, center=<0 0.0524 0.5991>, rotation=<rot 0.0001 0 -85>}]
 dyn_whistle_06 [{shapeType=Capsule, length=0.5472, radius=2, shapeName=dyn_whistle_06_capsule, center=<0 0.0542 0.6196>, rotation=<rot 0.0001 0 -85>}]
 dyn_whistle_07 [{
   shapeType=Capsule
   length=2.0195
   radius=2.3278
   shapeName=dyn_whistle_07_capsule
   center=<2.9676 0.0155 0.1771>
   rotation=<rot 5 -90 -80.0004>
  }]
 dyn_whistle_master [{shapeType=Capsule, length=0.505, radius=0.505, shapeName=dyn_whistle_master_capsule, rotation=<rot 0 -90 0>}]
 spine_03 [{
   shapeType=Capsule
   length=15.053
   radius=13.4011
   shapeName=spine_03_capsule
   center=<9.9076 -1.5662 -0.039>
   rotation=<rot 1.4384 -90 -0.5851>
  }]
 spine_04 [
  {
   shapeType=Capsule
   length=26.7916
   radius=9.8674
   shapeName=spine_04_capsule
   center=<10.6098 0.2486 -0.0889>
   rotation=<rot -0.0462 -90 -0.0196>
  }
  {
   shapeType=Capsule
   length=18.0628
   radius=7.7037
   shapeName=spine_04_capsule1
   center=<7.8298 3.7279 -0.0905>
   rotation=<rot -0.0462 -90 -0.0196>
  }
 ]
 spine_05 [
  {
   shapeType=Capsule
   length=11.7432
   radius=13.0985
   shapeName=spine_05_capsule
   center=<-7.0394 -0.9634 0.027>
   rotation=<rot -0.1112 -90 -0.0821>
  }
  {
   shapeType=Capsule
   length=12.556
   radius=13.3482
   shapeName=spine_05_capsule1
   center=<-12.2735 -1.4015 0.0043>
   rotation=<rot -0.1112 -90 -0.0821>
  }
 ]
}
constraints [
 {bone1Name=dyn_whistle_03, bone2Name=dyn_whistle_02, swing1LimitDegrees=30, swing2LimitDegrees=30, twistLimitDegrees=10}
 {bone1Name=dyn_whistle_04, bone2Name=dyn_whistle_03, swing1LimitDegrees=30, swing2LimitDegrees=30, twistLimitDegrees=10}
 {bone1Name=dyn_whistle_05, bone2Name=dyn_whistle_04, swing1LimitDegrees=30, swing2LimitDegrees=30, twistLimitDegrees=10}
 {bone1Name=dyn_whistle_06, bone2Name=dyn_whistle_05, swing1LimitDegrees=30, swing2LimitDegrees=30, twistLimitDegrees=10}
 {bone1Name=dyn_whistle_02, bone2Name=dyn_whistle_01, swing1LimitDegrees=15, swing2LimitDegrees=30, twistLimitDegrees=5}
 {bone1Name=dyn_whistle_07, bone2Name=dyn_whistle_06, swing1LimitDegrees=30, swing2LimitDegrees=30, twistLimitDegrees=10}
 {bone1Name=dyn_whistle_01, bone2Name=dyn_whistle_master, swing1LimitDegrees=15, swing2LimitDegrees=15, twistLimitDegrees=5}
 {
  bone1Name=spine_04
  bone2Name=spine_03
  swing1LimitDegrees=45
  swing1Motion=Free
  swing2LimitDegrees=45
  swing2Motion=Free
  twistLimitDegrees=45
  twistMotion=Free
 }
 {
  bone1Name=spine_05
  bone2Name=spine_03
  swing1LimitDegrees=45
  swing1Motion=Free
  swing2LimitDegrees=45
  swing2Motion=Free
  twistLimitDegrees=45
  twistMotion=Free
 }
]
