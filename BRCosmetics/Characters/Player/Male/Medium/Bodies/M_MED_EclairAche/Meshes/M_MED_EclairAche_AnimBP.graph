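# Post-copy AnimBP for the M_MED_EclairAche male body cosmetic, adding secondary motion to its zipper and collar.
# It copies the attached parent's pose through a Control Rig, simulates the zipper with AnimDynamics, and constrains the zipper master bone to weighted spine, clavicle and neck bones.
# A spine_05-relative RigidBody pass follows, then spring-driven AnimDynamics for each collar side.
# The event graph caches left and right collar dynamics curves into variables.
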AnimBlueprint /BRCosmetics/Characters/Player/Male/Medium/Bodies/M_MED_EclairAche/Meshes/M_MED_EclairAche_AnimBP
target_skeleton=&/M_MED_EclairAche_Skeleton
graphs [
 AnimGraph {nodes [
   n1 = Control Rig(Source=Copy Pose From Mesh() {bUseAttachedParent=true}) {controlRigAssetReference {blueprintRigClass=&/M_MED_EclairAche_CtrlRig.M_MED_EclairAche_CtrlRig_C}}
   n2 = Anim Dynamics - Bone: dyn_zipper_01(ComponentPose=Local To Component(LocalPose=@n1), Alpha=1, ExternalForce="0,0,0") {
    angularDampingOverride=0.999
    bGravityOverrideInSimSpace=true
    bOverrideAngularDamping=true
    bUseGravityOverride=true
    boundBone {boneName=dyn_zipper_01}
    componentAppliedLinearAccClamp=<1000 1000 1000>
    componentLinearAccScale=<0.5 0.5 0.5>
    componentLinearVelScale=<0.5 0.5 0.5>
    gravityOverride {z=-980}
    physicsBodyDefinitions [{
      boundBone {boneName=dyn_zipper_01}
      boxExtents=<10 10 10>
      constraintSetup {angularLimitsMax=<0 10 30>, angularLimitsMin=<0 -10 0>, angularTarget=<1 0 0>}
      localJointOffset=<1 0 0>
     }]
    relativeSpaceBone {boneName=spine_05}
    simSpaceSettings {maxAngularAcceleration=30, maxAngularVelocity=8, simSpaceAngularAlpha=1}
   }
   n3 = "Constraint - dyn_zipper_master (PPPPPPPP)"!(
    ComponentPose=@n2
    Alpha=1
    ConstraintWeights_0=0
    ConstraintWeights_1:float=0.0171
    ConstraintWeights_2:float=0.0761
    ConstraintWeights_3:float=0.7621
    ConstraintWeights_4:float=0.0124
    ConstraintWeights_5:float=0.0124
    ConstraintWeights_6:float=0.0365
    ConstraintWeights_7:float=0.0833
   ) -> (Pose) {
    boneToModify {boneName=dyn_zipper_master}
    constraintSetup [
     {targetBone {boneName=spine_02}}
     {targetBone {boneName=spine_03}}
     {targetBone {boneName=spine_04}}
     {targetBone {boneName=spine_05}}
     {targetBone {boneName=clavicle_l}}
     {targetBone {boneName=clavicle_r}}
     {targetBone {boneName=neck_01}}
     {targetBone {boneName=neck_02}}
    ]
    constraintWeights [1,1,1,1,1,1,1,1]
   }
   n4 = RigidBody(ComponentPose=@n3) {
    bClampLinearTranslationLimitToRefPose=true
    bForceDisableCollisionBetweenConstraintBodies=true
    bOverrideWorldGravity=true
    cachedBoundsScale=1.25
    overridePhysicsAsset=&/M_MED_EclairAche_Physics
    simulationSpace=BaseBoneSpace
    baseBoneRef {boneName=spine_05}
    overrideWorldGravity {z=-980}
    simSpaceSettings {
     maxAngularAcceleration=80
     maxAngularVelocity=8
     maxLinearAcceleration=1200
     maxLinearVelocity=650
     worldAlpha=1
     externalLinearDragV=<1 1 1>
    }
   }
   n5 = Anim Dynamics - Bone: dyn_collar_l_01(ComponentPose=@n4, Alpha=1, ExternalForce="0,0,0") {
    angularDampingOverride=0.9999
    angularSpringConstant=1500
    bAngularSpring=true
    bGravityOverrideInSimSpace=true
    bOverrideAngularDamping=true
    bUseGravityOverride=true
    alphaScaleBiasClamp {bMapRange=true, outRange {max=0, min=1}}
    boundBone {boneName=dyn_collar_l_01}
    componentAppliedLinearAccClamp=<5000 5000 5000>
    componentLinearAccScale=<0.5 0.5 0.5>
    componentLinearVelScale=<0.5 0.5 0.5>
    gravityOverride {z=-980}
    physicsBodyDefinitions [{
      boundBone {boneName=dyn_collar_l_01}
      boxExtents=<15 15 15>
      constraintSetup {angularLimitsMax=<0 10 75>, angularLimitsMin=<0 -10 -75>, angularTarget=<1 0 0>}
      localJointOffset=<-0.5 0 0>
     }]
    relativeSpaceBone {boneName=spine_05}
    simSpaceSettings {maxAngularAcceleration=30, maxAngularVelocity=8, simSpaceAngularAlpha=1}
   }
   n6 = Anim Dynamics - Bone: dyn_collar_r_01(ComponentPose=@n5, Alpha=1, ExternalForce="0,0,0") {
    angularDampingOverride=0.9999
    angularSpringConstant=1500
    bAngularSpring=true
    bGravityOverrideInSimSpace=true
    bOverrideAngularDamping=true
    bUseGravityOverride=true
    alphaScaleBiasClamp {bMapRange=true, outRange {max=0, min=1}}
    boundBone {boneName=dyn_collar_r_01}
    componentAppliedLinearAccClamp=<5000 5000 5000>
    componentLinearAccScale=<0.5 0.5 0.5>
    componentLinearVelScale=<0.5 0.5 0.5>
    gravityOverride {z=-980}
    physicsBodyDefinitions [{
      boundBone {boneName=dyn_collar_r_01}
      boxExtents=<15 15 15>
      constraintSetup {angularLimitsMax=<0 10 75>, angularLimitsMin=<0 -10 -75>, angularTarget=<1 0 0>}
      localJointOffset=<0.5 0 0>
     }]
    relativeSpaceBone {boneName=spine_05}
    simSpaceSettings {maxAngularAcceleration=30, maxAngularVelocity=8, simSpaceAngularAlpha=1}
   }
   Output Pose(Result=Component To Local(ComponentPose=@n6)) {layerGroup=None}
  ]}
 EventGraph {nodes [
   Event BlueprintUpdateAnimation()
   > Set collar_dyn_l(collar_dyn_l:double=GetCurveValue(CurveName=collar_dyn_l)) -> (Output_Get:double)
   > Set collar_dyn_r(collar_dyn_r:double=GetCurveValue(CurveName=collar_dyn_r)) -> (Output_Get:double)
  ]}
]
variables [
 collar_dyn_l {type=number}
 collar_dyn_r {type=number}
]
